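# Default Lit material for the snowball toy cosmetic, usable on skeletal meshes, that gives snow a layered, view-dependent look.
# Two samples of one noise texture use bump offsets at different reference planes to fake parallax depth.
# Their sum, weighted by tangent-space camera Z, blends two snow color parameters and fades a flat normal into the snow normal map.
# The combined noise also drives a scalar blend, likely roughness.
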Material /BRCosmetics/Abilities/Toys/Snowball/Materials/M_Snowball_01
graph {
 properties {bUsedWithMeshDeformer=true, bUsedWithSkeletalMesh=true, pixelDepthOffsetMode=PDOM_Legacy, usedShadingModels=Default Lit}
 nodes [
  n1 = Transform(None=CameraVectorWS()) {transformSourceType=TRANSFORMSOURCE_World, transformType=TRANSFORM_Tangent}
  n2 = TextureCoordinate() {uTiling=2, vTiling=2}
  n3 = Constant()
  n4 = TextureSample(UVs=BumpOffset(Coordinate=@n2, Height=@n3)) {texture=/Game/Abilities/Toys/Snowball/Textures/T_Snowball_Noise_01}
  n5 = TextureSample(UVs=BumpOffset(Coordinate=@n2, Height=@n3) {referencePlane=0.2}) {texture=&T_Snowball_Noise_01}
  n6 = Add(A=@n4.R, B=@n5.R)
  n7 = Clamp(None=Multiply(A=ComponentMask(None=@n1) {b=true}, B=@n6))
  LinearInterpolate(
   A=VectorParameter() {parameterName=SnowColor_A, defaultValue=<rgba 1 1 1 0>}.RGB
   B=VectorParameter() {parameterName=SnowColor_B, defaultValue=<rgba 0.2989 0.732 0.9896 0>}.RGB
   Alpha=@n7
  )
  LinearInterpolate(Alpha=Multiply(A=@n6, B=Multiply(A=@n4.R, B=@n5.R))) {constA=1, constB=0.2}
  LinearInterpolate(
   A=Constant3Vector() {constant=<rgba 0 0 1 0>}
   B=TextureSample() {samplerType=SAMPLERTYPE_Normal, texture=/Game/Abilities/Toys/Snowball/Textures/T_Snowball_Normals}.RGB
   Alpha=@n7
  )
 ]
}
